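# Mono one-shot MetaSound for toy ball impacts and rolling, defaulting to soccer ball hits.
# Randomly gated ball-hit triggers start an enveloped roll loop below the magnitude threshold or random impacts above it, with speed selecting soft or hard.
# Player hits bypass that gate and can force hard impacts.
# Speed scales all layers and hit pitch, smoothed elevation shapes roll volume, and OnFinished fires after a resettable delay once hit and roll end.
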MetaSoundSource /Game/Athena/Sounds/Toys/MSS_Ball_Collision_Base
inputs [
 Gain.HardHit {type=Float}
 Gain.Rolling {type=Float}
 Gain.SoftHit {type=Float}
 GateDelayMax {type=Time}
 GateDelayMin {type=Time}
 Movement.BallHitMagnitude {type=Float}
 Movement.HardHitThreshold {type=Float}
 Movement.MagnitudeThreshold {type=Float}
 Movement.MapSpeedMax {type=Float}
 Movement.Speed {type=Float}
 Movement.SpeedAzimuth {type=Float}
 Movement.SpeedElevation {type=Float}
 PitchShiftMax {type=Float}
 PitchShiftMin {type=Float}
 PlayerHitAlwaysHard {type=Bool}
 RetriggerDelay {type=Time}
 Sound.HardHit {type=WaveAsset:Array}
 Sound.RollingSound {type=WaveAsset}
 Sound.SoftHit {type=WaveAsset:Array}
 Trigger.BallHit {type=Trigger}
 Trigger.PlayerHit {type=Trigger}
 UE.Source.OnPlay {type=Trigger}
]
interfaces [Fort.Movement:1.0, UE.OutputFormat.Mono:1.0, UE.Source.OneShot:1.0, UE.Source:1.1]
outputs [
 Out {type=Trigger}
 "UE.OutputFormat.Mono.Audio:0" {type=Audio}
 UE.Source.OneShot.OnFinished {type=Trigger}
]
pages [Default {
  input_defaults {
   Gain.HardHit=0.9
   Gain.Rolling=1
   Gain.SoftHit=1
   GateDelayMax=0.2
   GateDelayMin=0.05
   Movement.BallHitMagnitude=0
   Movement.HardHitThreshold=600
   Movement.MagnitudeThreshold=10
   Movement.MapSpeedMax=400
   Movement.Speed=59.9999
   Movement.SpeedAzimuth=0
   Movement.SpeedElevation=0
   PitchShiftMax=2
   PitchShiftMin=-3
   PlayerHitAlwaysHard=null
   RetriggerDelay=1
   Sound.RollingSound=/Game/Sounds/Creative/Gadgets/Ball/Ball_Roll_Loop_01
   Trigger.BallHit=null
   Trigger.PlayerHit=null
   UE.Source.OnPlay=null
   Sound.HardHit [
    /Game/Athena/Sounds/Toys/SoccerBall/SoccerBall_Hit_01
    /Game/Athena/Sounds/Toys/SoccerBall/SoccerBall_Hit_02
    /Game/Athena/Sounds/Toys/SoccerBall/SoccerBall_Hit_07
    /Game/Athena/Sounds/Toys/SoccerBall/SoccerBall_Hit_08
    /Game/Athena/Sounds/Toys/SoccerBall/SoccerBall_Hit_09
    /Game/Athena/Sounds/Toys/SoccerBall/SoccerBall_Hit_10
    /Game/Athena/Sounds/Toys/SoccerBall/SoccerBall_Hit_11
    /Game/Athena/Sounds/Toys/SoccerBall/SoccerBall_Hit_12
    /Game/Athena/Sounds/Toys/SoccerBall/SoccerBall_Hit_13
    /Game/Athena/Sounds/Toys/SoccerBall/SoccerBall_Hit_14
   ]
   Sound.SoftHit [
    /Game/Athena/Sounds/Toys/SoccerBall/SoccerBall_Hit_03
    &SoccerBall_Hit_01
    /Game/Athena/Sounds/Toys/SoccerBall/SoccerBall_Hit_05
    /Game/Athena/Sounds/Toys/SoccerBall/SoccerBall_Hit_06
    &SoccerBall_Hit_07
    &SoccerBall_Hit_08
    &SoccerBall_Hit_09
    &SoccerBall_Hit_10
    &SoccerBall_Hit_11
    &SoccerBall_Hit_12
    &SoccerBall_Hit_13
    &SoccerBall_Hit_14
   ]
  }
  nodes [
   EnvelopeFollower_0 = UE.Envelope Follower(In=UE.ConversionFloatToAudio(In=MapRange.MapRange:Float(In=^Movement.SpeedElevation) {also {defaults {In Range A=-90, In Range B=90, Out Range A=1, Out Range B=-1}}}.Out Value).Out) {also {defaults {Release Time=0.2}}}
   n1 = VariableDeferredAccessor.Bool(Variable=InitVariable.Bool().Variable)
   n2 = TriggerCompare.Clamp:Bool(Compare=^Trigger.BallHit, A=@n1.Value) {also {defaults {B="<MetaSoundBuilderResult.SUCCEEDED: 0>"}}}
   TriggerAnyTriggerAny2_1 = "TriggerAny.Trigger Any (2)"(In 0=^UE.Source.OnPlay, In 1=@n2.False)
   n3 = TriggerCompare.Clamp:Float(
    A=^Movement.BallHitMagnitude
    B=^Movement.MagnitudeThreshold
    Compare=TriggerCompare.Clamp:Float(A=^Movement.Speed, Compare=@TriggerAnyTriggerAny2_1.Out) {also {defaults {B=0, Type=3}}}.True
   ) {also {defaults {Type=2}}}
   TriggerSequence2_0 = "UE.Trigger Sequence (2)"(Reset=@n3.False, In=@n3.True)
   ADSREnvelopeADSREnvelope_0 = ADSR Envelope.ADSR Envelope:Float(Trigger Release=UE.Reroute(Value=@n3.False).Value, Trigger Attack=@TriggerSequence2_0.Out 0) {also {defaults {Attack Curve=1.5, Attack Time=0.25, Hard Reset="<MetaSoundBuilderResult.SUCCEEDED: 0>", Release Time=0.5, Sustain Level=1}}}
   WavePlayer_0 = UE.Wave Player:Mono(
    Wave Asset=^Sound.RollingSound
    Play=@ADSREnvelopeADSREnvelope_0.On Attack Triggered
    Stop=@ADSREnvelopeADSREnvelope_0.On Done
    Start Time=&MSP_WaveGetRandomStartTime(Sound=^Sound.RollingSound).Value
   ) {also {defaults {Loop="<MetaSoundBuilderResult.SUCCEEDED: 0>"}, version [1,1]}}
   n4 = UE.Multiply:Audio by Float(AdditionalOperands=@ADSREnvelopeADSREnvelope_0.Out Envelope, PrimaryOperand=@WavePlayer_0.Out Mono)
   n5 = VariableAccessor.Float(Variable=VariableMutator.Float(Variable=InitVariable.Float().Variable, Value=MapRange.MapRange:Float(In Range B=^Movement.MapSpeedMax, In=^Movement.Speed).Out Value).Variable)
   n6 = VariableAccessor.Float(Variable=@n5.Variable)
   n7 = VariableAccessor.Float(Variable=@n6.Variable)
   n8 = &MSP_Branch(Value=^PlayerHitAlwaysHard, In=^Trigger.PlayerHit)
   TriggerAnyTriggerAny2_4 = "TriggerAny.Trigger Any (2)"(In 1=^UE.Source.OnPlay, In 0=@n2.False)
   n9 = TriggerCompare.Clamp:Float(A=^Movement.BallHitMagnitude, B=^Movement.MagnitudeThreshold, Compare=@TriggerAnyTriggerAny2_4.Out) {also {defaults {Type=3}}}
   TriggerAnyTriggerAny2_0 = "TriggerAny.Trigger Any (2)"(In 0=@n8.False, In 1=@n9.True)
   n10 = TriggerCompare.Clamp:Float(B=^Movement.HardHitThreshold, A=^Movement.Speed, Compare=@TriggerAnyTriggerAny2_0.Out) {also {defaults {Type=3}}}
   ArrayRandomGet_0 = Array.Random Get:WaveAsset:Array(In Array=^Sound.SoftHit, Next=@n10.False) {also {defaults {Weights []}}}
   n11 = VariableAccessor.Float(Variable=VariableMutator.Float(
     Variable=InitVariable.Float().Variable
     Value=MapRange.MapRange:Float(Out Range B=^PitchShiftMax, Out Range A=^PitchShiftMin, In=@n6.Value).Out Value
    ).Variable)
   WavePlayer_1 = UE.Wave Player:Mono(Play=@ArrayRandomGet_0.On Next, Wave Asset=@ArrayRandomGet_0.Value, Pitch Shift=VariableAccessor.Float(Variable=@n11.Variable).Value) {also {version [1,1]}}
   TriggerAnyTriggerAny2_2 = "TriggerAny.Trigger Any (2)"(In 0=@n8.True, In 1=@n10.True)
   ArrayRandomGet_1 = Array.Random Get:WaveAsset:Array(In Array=^Sound.HardHit, Next=@TriggerAnyTriggerAny2_2.Out)
   WavePlayer_2 = UE.Wave Player:Mono(Play=@ArrayRandomGet_1.On Next, Wave Asset=@ArrayRandomGet_1.Value, Pitch Shift=@n11.Value) {also {version [1,1]}}
   AudioMixerAudioMixerMono3_0 = "AudioMixer.Audio Mixer (Mono, 3)"(
    Gain 0=^Gain.HardHit
    Gain 2=^Gain.Rolling
    Gain 1=^Gain.SoftHit
    In 2=UE.Multiply:Audio by Float(
     PrimaryOperand=UE.Multiply:Audio by Float(AdditionalOperands=@EnvelopeFollower_0.Envelope, PrimaryOperand=@n4.Out).Out
     AdditionalOperands=@n7.Value
    ).Out
    In 1=UE.Multiply:Audio by Float(AdditionalOperands=@n5.Value, PrimaryOperand=@WavePlayer_1.Out Mono).Out
    In 0=UE.Multiply:Audio by Float(AdditionalOperands=VariableAccessor.Float(Variable=@n7.Variable).Value, PrimaryOperand=@WavePlayer_2.Out Mono).Out
   )
   InitVariable.Float()
   InitVariable.Trigger()
   InitVariable.Trigger()
   InitVariable.Trigger()
   n12 = InitVariable.Trigger()
   TriggerAnyTriggerAny2_3 = "TriggerAny.Trigger Any (2)"(In 1=@WavePlayer_1.On Finished, In 0=@WavePlayer_2.On Finished)
   n13 = InitVariable.Trigger()
   TriggerAccumulatorTriggerAccumulate2_0 = "TriggerAccumulator.Trigger Accumulate (2)"(
    In 0=VariableAccessor.Trigger(Variable=VariableMutator.Trigger(Variable=@n12.Variable, Value=@TriggerAnyTriggerAny2_3.Out).Variable).Value
    In 1=VariableAccessor.Trigger(Variable=VariableMutator.Trigger(Variable=@n13.Variable, Value=@WavePlayer_0.On Finished).Variable).Value
   ) {also {defaults {Auto Reset="<MetaSoundBuilderResult.SUCCEEDED: 0>"}}}
   n14 = InitVariable.Trigger()
   TriggerDelay_1 = UE.Trigger Delay(
    Delay Time=^RetriggerDelay
    In=@TriggerAccumulatorTriggerAccumulate2_0.Out
    Reset=VariableAccessor.Trigger(Variable=VariableMutator.Trigger(Variable=@n14.Variable, Value=@n9.True).Variable).Value
   )
   n15 = VariableDeferredAccessor.Trigger(Variable=InitVariable.Trigger().Variable)
   n16 = Value.Value:Bool(Set=@n2.False, Reset=@n15.Value)
   VariableMutator.Bool(Value=@n16.Output Value, Variable=@n1.Variable)
   n17 = UE.RandomTime(Max=^GateDelayMax, Min=^GateDelayMin, Next=@n16.On Set)
   TriggerDelay_0 = UE.Trigger Delay(In=@n17.On Next, Delay Time=@n17.Value)
   VariableMutator.Trigger(Value=@TriggerDelay_0.Out, Variable=@n15.Variable)
  ]
  links [{dst="@UE.OutputFormat.Mono.Audio:0", src=AudioMixerAudioMixerMono3_0.Out}, {dst="@UE.Source.OneShot.OnFinished", src=TriggerDelay_1.Out}]
 }]
properties {
 assetImportData=&self:AssetImportData
 attenuationSettings=/Game/Sounds/Attenuation/Fort_Ball_Soccer_Bounce_Athena_Att
 maxDistance=3010
 sampleRate=48000
 soundClassObject=&Fort_Ambient_Props_Athena
}
