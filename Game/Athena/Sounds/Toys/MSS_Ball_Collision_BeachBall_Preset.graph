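# MetaSound preset that plays beach ball collision sounds for the Athena toy, a mono one-shot source built on MSS_Ball_Collision_Base.
# It only overrides inputs: hard and soft hit wave arrays, gains, pitch range, retrigger and gate timing, and speed thresholds, while the parent holds the graph logic.
# Hit triggers and movement values are exposed as inputs, and the rolling sound is left empty.
# It reuses a soccer ball bounce attenuation.
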MetaSoundSource /Game/Athena/Sounds/Toys/MSS_Ball_Collision_BeachBall_Preset
preset_parent=&/MSS_Ball_Collision_Base
inputs [
 Gain.HardHit {type=Float}
 Gain.Rolling {type=Float}
 Gain.SoftHit {type=Float}
 GateDelayMax {type=Time}
 GateDelayMin {type=Time}
 Movement.BallHitMagnitude {type=Float}
 Movement.HardHitThreshold {type=Float}
 Movement.MagnitudeThreshold {type=Float}
 Movement.MapSpeedMax {type=Float}
 Movement.Speed {type=Float}
 Movement.SpeedAzimuth {type=Float}
 Movement.SpeedElevation {type=Float}
 PitchShiftMax {type=Float}
 PitchShiftMin {type=Float}
 PlayerHitAlwaysHard {type=Bool}
 RetriggerDelay {type=Time}
 Sound.HardHit {type=WaveAsset:Array}
 Sound.RollingSound {type=WaveAsset}
 Sound.SoftHit {type=WaveAsset:Array}
 Trigger.BallHit {type=Trigger}
 Trigger.PlayerHit {type=Trigger}
 UE.Source.OnPlay {type=Trigger}
]
interfaces [Fort.Movement:1.0, UE.OutputFormat.Mono:1.0, UE.Source.OneShot:1.0, UE.Source:1.1]
outputs [
 Out {type=Trigger}
 "UE.OutputFormat.Mono.Audio:0" {type=Audio}
 UE.Source.OneShot.OnFinished {type=Trigger}
]
pages [Default {input_defaults {
   Gain.HardHit=0.9
   Gain.Rolling=1
   Gain.SoftHit=1
   GateDelayMax=0.1
   GateDelayMin=0.05
   Movement.BallHitMagnitude=0
   Movement.HardHitThreshold=600
   Movement.MagnitudeThreshold=10
   Movement.MapSpeedMax=400
   Movement.Speed=59.9999
   Movement.SpeedAzimuth=0
   Movement.SpeedElevation=0
   PitchShiftMax=0
   PitchShiftMin=-3
   PlayerHitAlwaysHard=null
   RetriggerDelay=1
   Sound.RollingSound=null
   Trigger.BallHit=null
   Trigger.PlayerHit=null
   UE.Source.OnPlay=null
   Sound.HardHit [
    /Game/Athena/Sounds/Toys/BeachBall/BeachBall_HeavyImpact_04
    /Game/Athena/Sounds/Toys/BeachBall/BeachBall_HeavyImpact_05
    /Game/Athena/Sounds/Toys/BeachBall/BeachBall_HeavyImpact_06
   ]
   Sound.SoftHit [
    &/Toys_BeachBall_Bounce_10
    &/Toys_BeachBall_Bounce_11
    &/Toys_BeachBall_Bounce_12
    &/Toys_BeachBall_Bounce_14
   ]
  }}]
properties {
 assetImportData=&self:AssetImportData
 attenuationSettings=/Game/Sounds/Attenuation/Fort_Ball_Soccer_Bounce_Athena_Att
 maxDistance=3010
 sampleRate=48000
 soundClassObject=&Fort_Ambient_Props_Athena
 volume=2
}
