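# Sound cue for the looping in-air sound of the snowball toy, routed through the gadgets sound class with toy-hit attenuation.
# A looping wave player feeds a modulator that randomizes pitch upward, then a smoothed Doppler node for motion-based pitch shift.
# An enveloper at the root ramps volume up from half to full as the sound starts, and the cue itself doubles the base pitch.
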SoundCue /Game/Athena/Sounds/Toys/Toy_Snowball_InAirLoop_Cue
root=Enveloper_0
nodes [
 {id=Doppler_0, children [Modulator_0], properties {bUseSmoothing=true, smoothingInterpSpeed=1}}
 {
  id=Enveloper_0
  children [Doppler_0]
  properties {pitchCurve {editorCurveData {keys [{time=0}]}}, volumeCurve {editorCurveData {keys [<at 0 0.5>,<at 1 1>]}}}
 }
 {id=Modulator_0, children [WavePlayer_0], properties {pitchMax=1.5, pitchMin=1}}
 {id=WavePlayer_0, properties {bLooping=true, soundWaveAssetPtr=&/Toy_Snowball_InAirLoop}}
]
properties {
 attenuationSettings=/Game/Sounds/Attenuation/Fort_ToyHit_Att
 maxDistance=2800
 pitchMultiplier=2
 soundClassObject=&Fort_Gadgets
 volumeMultiplier=1
}
