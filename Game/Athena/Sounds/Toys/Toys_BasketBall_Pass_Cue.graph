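# Sound cue for the pass sound of the basketball toy in Athena.
# A Random node picks between two pass wave variations with equal weight.
# It borrows the emote attenuation and emote music sound class, with volume raised slightly above default.
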SoundCue /Game/Athena/Sounds/Toys/Toys_BasketBall_Pass_Cue
root=Random_0
nodes [
 {id=Random_0, children [WavePlayer_0,WavePlayer_1], properties {weights [1,1]}}
 {id=WavePlayer_0, properties {soundWaveAssetPtr=&/Toys_BasketBall_Pass_06}}
 {id=WavePlayer_1, properties {soundWaveAssetPtr=&/Toys_BasketBall_Pass_05}}
]
properties {
 attenuationSettings=&Fort_Emotes_Athena_Att
 maxDistance=1600
 soundClassObject=&Fort_Music_Emote
 volumeMultiplier=1.25
}
