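# Sound cue for the basketball toy's scoring net sound, layering a net sound with a score pop.
# A mixer combines a modulated random pick of two net waves at full level with a quieter random pick of three small LTM score pops.
# A fixed half-second delay makes the pop follow the net sound.
# It uses ball-bounce attenuation and the music emote sound class.
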SoundCue /Game/Athena/Sounds/Toys/Toys_BasketBall_Score_Net_Cue
root=Mixer_0
nodes [
 {id=Delay_1, children [Random_3], properties {delayMax=0.5, delayMin=0.5}}
 {id=Mixer_0, children [Delay_1,Modulator_0], properties {inputVolume [0.5,1]}}
 {id=Modulator_0, children [Random_1]}
 {id=Random_1, children [WavePlayer_2,WavePlayer_3], properties {weights [1,1]}}
 {id=Random_3, children [WavePlayer_4,WavePlayer_8,WavePlayer_9], properties {weights [1,1,1]}}
 {id=WavePlayer_2, properties {soundWaveAssetPtr=&/Toys_BasketBall_Score_Net_23}}
 {id=WavePlayer_3, properties {soundWaveAssetPtr=&/Toys_BasketBall_Score_Net_24}}
 {id=WavePlayer_4, properties {soundWaveAssetPtr=/Game/Athena/Sounds/LTMs/Score/ScoreLTM_InGamePop_Little_01}}
 {id=WavePlayer_8, properties {soundWaveAssetPtr=/Game/Athena/Sounds/LTMs/Score/ScoreLTM_InGamePop_Little_02}}
 {id=WavePlayer_9, properties {soundWaveAssetPtr=/Game/Athena/Sounds/LTMs/Score/ScoreLTM_InGamePop_Little_03}}
]
properties {
 attenuationSettings=/Game/Sounds/Attenuation/Fort_Ball_Bounce_Athena_Att
 maxDistance=7000
 soundClassObject=&Fort_Music_Emote
 volumeMultiplier=1.25
}
