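# Sound cue for a beach ball toy hitting water, routed to the water interaction sound class with ball bounce attenuation.
# A mixer layers a randomly chosen small water impact, held back by a brief fixed delay, with a quieter randomly chosen beach ball bounce.
# A modulator at the root varies the combined output, which plays at low volume with slightly raised pitch.
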SoundCue /Game/Athena/Sounds/Toys/Toys_BeachBall_Water_Impact_Cue
root=Modulator_0
nodes [
 {id=Delay_0, children [Random_1], properties {delayMax=0.04, delayMin=0.04}}
 {id=Mixer_0, children [Delay_0,Random_2], properties {inputVolume [1,0.65]}}
 {id=Modulator_0, children [Mixer_0]}
 {id=Random_1, children [WavePlayer_23, WavePlayer_24, WavePlayer_25], properties {weights [1,1,1]}}
 {id=Random_2, children [WavePlayer_8,WavePlayer_9], properties {weights [1,1]}}
 {id=WavePlayer_23, properties {soundWaveAssetPtr=/Game/Sounds/Apollo/HeldObjects/HeldObjects_Small_ImpactWater_01}}
 {id=WavePlayer_24, properties {soundWaveAssetPtr=/Game/Sounds/Apollo/HeldObjects/HeldObjects_Small_ImpactWater_02}}
 {id=WavePlayer_25, properties {soundWaveAssetPtr=/Game/Sounds/Apollo/HeldObjects/HeldObjects_Small_ImpactWater_03}}
 {id=WavePlayer_8, properties {soundWaveAssetPtr=&/Toys_BeachBall_Bounce_13}}
 {id=WavePlayer_9, properties {soundWaveAssetPtr=&/Toys_BeachBall_Bounce_15}}
]
properties {
 attenuationSettings=/Game/Sounds/Attenuation/Fort_Ball_Bounce_Athena_Att
 maxDistance=7000
 pitchMultiplier=1.1
 soundClassObject=/Game/Sounds/FortSoundClassesAndModes/Fort_Water_Interaction
 volumeMultiplier=0.15
}
