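# Sound cue for the golf ball toy's bounce sound in Athena.
# A random node picks between two equally weighted bounce waves and feeds a modulator at its default settings.
# The cue uses a shared ball-bounce attenuation asset, a boosted volume multiplier and the Fort_Music_Emote sound class.
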SoundCue /Game/Athena/Sounds/Toys/Toys_GolfBall_Bounce
root=Modulator_0
nodes [
 {id=Modulator_0, children [Random_1]}
 {id=Random_1, children [WavePlayer_0,WavePlayer_4], properties {weights [1,1]}}
 {id=WavePlayer_0, properties {soundWaveAssetPtr=&/Toys_GolfBall_Bounce_20}}
 {id=WavePlayer_4, properties {soundWaveAssetPtr=&/Toys_GolfBall_Bounce_17}}
]
properties {
 attenuationSettings=/Game/Sounds/Attenuation/Fort_Ball_Bounce_Athena_Att
 maxDistance=7000
 soundClassObject=&Fort_Music_Emote
 volumeMultiplier=1.25
}
