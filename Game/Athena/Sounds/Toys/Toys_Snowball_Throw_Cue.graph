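# Sound cue that plays the throw sound for the Athena snowball toy.
# A single wave player feeds a Modulator root, which varies each playback with its random pitch range widened on the low end.
# The cue is boosted in volume, uses the Athena emote attenuation for distance falloff, and is assigned to the emote music sound class.
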SoundCue /Game/Athena/Sounds/Toys/Toys_Snowball_Throw_Cue
root=Modulator_0
nodes [
 {id=Modulator_0, children [WavePlayer_2], properties {pitchMin=0.9}}
 {id=WavePlayer_2, properties {soundWaveAssetPtr=&/Toy_Snowball_Throw_01}}
]
properties {
 attenuationSettings=&Fort_Emotes_Athena_Att
 maxDistance=1600
 soundClassObject=&Fort_Music_Emote
 volumeMultiplier=1.25
}
